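# Physics asset for the dangling whistle part on the medium female football referee character.
# Four capsule bodies on the dyn_whistle bones form a chain hanging from spine_05, each joint carrying the same narrow swing and twist angle limits.
# Capsules on spine_01 through spine_05 are joined by angularly locked constraints, forming a rigid torso volume alongside the whistle chain.
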PhysicsAsset /BRCosmetics/Characters/Player/Female/Medium/Bodies/F_MED_Football_20_Referee_BLK/Meshes/Parts/F_MED_Ref_Whistle_Physics
bodies {
 dyn_whistle_01 [{shapeType=Capsule, length=6.1483, radius=1.7222, shapeName=dyn_whistle_01_capsule, center=<0 3.1082 0>}]
 dyn_whistle_02 [{shapeType=Capsule, length=4.4678, radius=2.4017, shapeName=dyn_whistle_02_capsule, center=<3.56 2.4254 0>}]
 dyn_whistle_03 [{shapeType=Capsule, length=4.222, radius=1.8663, shapeName=dyn_whistle_03_capsule, center=<0 0.5443 0>}]
 dyn_whistle_04 [{
   shapeType=Capsule
   length=0.67
   radius=2.8648
   shapeName=dyn_whistle_04_capsule
   center=<3.3803 0.9724 -0.8159>
   rotation=<rot -50 -89.9998 90.0001>
  }]
 spine_01 [{shapeType=Capsule, length=21, radius=4.755, shapeName=spine_01_capsule, center=<9.9943 8.0226 0>}]
 spine_02 [{shapeType=Capsule, length=21, radius=4.755, shapeName=spine_02_capsule, center=<7.7011 9.0119 0>}]
 spine_03 [{shapeType=Capsule, length=21, radius=4.755, shapeName=spine_03_capsule, center=<4.8654 9.5567 0>}]
 spine_04 [{shapeType=Capsule, length=21, radius=4.755, shapeName=spine_04_capsule, center=<2.1032 8.4719 0>}]
 spine_05 [{shapeType=Capsule, length=19.3381, radius=8.3484, shapeName=spine_05_capsule, center=<0 1.621 0>}]
}
constraints [
 {
  bone1Name=spine_05
  bone2Name=spine_04
  swing1LimitDegrees=45
  swing1Motion=Locked
  swing2LimitDegrees=45
  swing2Motion=Locked
  twistLimitDegrees=45
  twistMotion=Locked
 }
 {
  bone1Name=spine_02
  bone2Name=spine_01
  swing1LimitDegrees=45
  swing1Motion=Locked
  swing2LimitDegrees=45
  swing2Motion=Locked
  twistLimitDegrees=45
  twistMotion=Locked
 }
 {
  bone1Name=spine_03
  bone2Name=spine_02
  swing1LimitDegrees=45
  swing1Motion=Locked
  swing2LimitDegrees=45
  swing2Motion=Locked
  twistLimitDegrees=45
  twistMotion=Locked
 }
 {
  bone1Name=spine_04
  bone2Name=spine_03
  swing1LimitDegrees=45
  swing1Motion=Locked
  swing2LimitDegrees=45
  swing2Motion=Locked
  twistLimitDegrees=45
  twistMotion=Locked
 }
 {bone1Name=dyn_whistle_01, bone2Name=spine_05, swing1LimitDegrees=20, swing2LimitDegrees=10, twistLimitDegrees=5}
 {bone1Name=dyn_whistle_02, bone2Name=dyn_whistle_01, swing1LimitDegrees=20, swing2LimitDegrees=10, twistLimitDegrees=5}
 {bone1Name=dyn_whistle_03, bone2Name=dyn_whistle_02, swing1LimitDegrees=20, swing2LimitDegrees=10, twistLimitDegrees=5}
 {bone1Name=dyn_whistle_04, bone2Name=dyn_whistle_03, swing1LimitDegrees=20, swing2LimitDegrees=10, twistLimitDegrees=5}
]
