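# Post-process-style Anim Blueprint for the female football referee (BLK) face accessory, adding secondary motion to its whistle.
# It copies the attached parent mesh's pose, simulates the whistle with a RigidBody node in root-bone space, then additively offsets dyn_whistle_02.
# Each update sets dynAlpha and frontEndMove by whether the cosmetic is displayed in a frontend context, though the AnimGraph shows no use of either.
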AnimBlueprint /BRCosmetics/Characters/Player/Female/Medium/Bodies/F_MED_Football_20_Referee_BLK/Meshes/Parts/F_MED_Football_20_Referee_BLK_FAcc_ABP
target_skeleton=&/F_MED_Football_20_Referee__BLK_FaceAcc
graphs [
 AnimGraph {nodes [
   n1 = RigidBody(ComponentPose=Local To Component(LocalPose=Copy Pose From Mesh() {bUseAttachedParent=true})) {
    bOverrideWorldGravity=true
    overridePhysicsAsset=&/F_MED_Ref_Whistle_Physics
    simulationSpace=BaseBoneSpace
    alphaBoolBlend {blendOption=Linear}
    baseBoneRef {boneName=root}
    componentAppliedLinearAccClamp=<3000 3000 3000>
    componentLinearAccScale=<0.2 0.2 0.2>
    componentLinearVelScale=<0.6 0.6 0.6>
    overrideWorldGravity {z=-980}
   }
   n2 = "Transform (Modify) Bone - Bone: dyn_whistle_02"(ComponentPose=@n1, Alpha=1, Translation="0,1,1", Rotation="0,0,0", Scale="1,1,1") {translationMode=BMM_Additive, alphaBoolBlend {blendOption=Linear}, boneToModify {boneName=dyn_whistle_02}, translation {y=1, z=1}}
   Output Pose(Result=Component To Local(ComponentPose=@n2)) {layerGroup=None}
  ]}
 EventGraph {nodes [
   Event BlueprintUpdateAnimation()
   > n1 = Branch(Condition=IsCosmeticObjectDisplayedInFrontendContext!(CosmeticObject:Object Reference) -> (ReturnValue:bool))
   > Set dynAlpha(dynAlpha:double=0.5) -> (Output_Get:double)
   > Set frontEndMove(frontEndMove:double=1) -> (Output_Get:double)
   Set dynAlpha(execute=@n1.else, dynAlpha=0.8)
   > Set frontEndMove(frontEndMove=0)
   TryGetPawnOwner()
  ]}
]
variables [
 dynAlpha {type=number}
 frontEndMove {type=number}
]
